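# Physics asset for the Doggus prototype's arm mesh, giving the left arm simple collision and joint bodies.
# Three rotated boxes cover the upper arm, lower arm and hand, chained by two constraints.
# The elbow has all swing and twist axes set free, so its stored 45-degree limits do not apply, while the wrist keeps default motion modes with 45-degree limit values.
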PhysicsAsset /Figment_S10/Athena/Prototype/Blueprints/Doggus/Meshes/M_MED_Doggus_Arm_Physics
bodies {
 hand_l [{shapeType=Box, extentX=11.8725, extentY=13.3912, extentZ=17.8471, center=<7.7861 -1.8408 0.5903>, rotation=<rot 0 86.6628 -70.2722>}]
 lowerarm_l [{shapeType=Box, extentX=19.8931, extentY=17.6918, extentZ=27.0198, center=<8.9526 0.3327 -0.8634>, rotation=<rot 0 90.3513 -88.4227>}]
 upperarm_l [{shapeType=Box, extentX=15.01, extentY=32.51, extentZ=14.51, center=<9.8779 0.0443 0.2627>, rotation=<rot 0 -86.2753 0>}]
}
constraints [
 {bone1Name=hand_l, bone2Name=lowerarm_l, swing1LimitDegrees=45, swing2LimitDegrees=45, twistLimitDegrees=45}
 {
  bone1Name=lowerarm_l
  bone2Name=upperarm_l
  swing1LimitDegrees=45
  swing1Motion=Free
  swing2LimitDegrees=45
  swing2Motion=Free
  twistLimitDegrees=45
  twistMotion=Free
 }
]
